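# Additive, unlit, two-sided lightning material for Storm Soldier pickaxe FX, usable on beam trails.
# Two offset samples of a lightning sheet on swizzled, scaled UV1 combine into a mask indexing a curve-atlas color ramp.
# The squared ramp is multiplied by particle color and dynamic AlphaMaster, then passed to MF_FX_EmissiveAdjustments.
# Separate camera-relative cross-product offset math and a constant-color multiply show no downstream connection.
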Material /BRCosmetics/Weapons/FORT_Melee/Pickaxe_StormSoldier/FX/M_Lightning
graph {
 properties {
  bUsedWithBeamTrails=true
  blendMode=BLEND_Additive
  pixelDepthOffsetMode=PDOM_Legacy
  shadingModel=MSM_Unlit
  twoSided=true
  usedShadingModels=Unlit
 }
 nodes [
  n1 = Multiply(
   A=MaterialFunctionCall(XY=TextureCoordinate() {coordinateIndex=1}) {materialFunction=&Swizzle}.YX
   B=Constant2Vector() {g=1, r=0.1}
  )
  n2 = Add(A=@n1, B=Constant2Vector() {r=0.1})
  n3 = Add(
   A=TextureSample(UVs=@n1) {texture=/Game/Athena/HUD/Materials/T-UI-LightningSheet}.R
   B=Multiply(A=TextureSample(UVs=@n2) {texture=&T-UI-LightningSheet}.R) {constB=0.1}
  )
  n4 = CurveAtlasRowParameter(CurveTime=@n3) {
   atlas=&AthenaCurveAtlas01
   curve=/Game/Athena/LinearColorCurves/Curve_Ramps
   defaultValue=95
   parameterName=Param2
  }
  n5 = Multiply(A=@n4, B=@n4)
  n6 = ParticleColor()
  MaterialFunctionCall(Emissive=Multiply(A=Multiply(A=@n5, B=@n6.RGB), B=DynamicParameter() {paramNames [AlphaMaster,Param2,Param3,Param4]}.AlphaMaster)) {materialFunction=&MF_FX_EmissiveAdjustments}
  n7 = ComponentMask(None=TextureCoordinate()) {g=true}
  n8 = Transform(None=Constant3Vector() {constant=<rgba 1 0 0 0>})
  n9 = Transform(None=Constant3Vector() {constant=<rgba 0 0 1 0>}) {transformSourceType=TRANSFORMSOURCE_Camera}
  Multiply(A=Multiply(A=Subtract(A=@n7) {constB=0.5}, B=CrossProduct(A=@n8, B=@n9))) {constB=50}
  Multiply(A=Constant3Vector() {constant=<rgba 128 0 60.4464 1>}, B=@n3)
 ]
}
