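# Niagara system for the lightning effect on the Storm Soldier pickaxe cosmetic, with two emitters named Long_bolts and Light.
# System-level scripts use SystemState, SampleSkeletalMesh with a skeletal mesh data interface, two Set Variables assignments and a TransformPosition dynamic input.
# No renderers or user parameters are listed.
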NiagaraSystem /BRCosmetics/Weapons/FORT_Melee/Pickaxe_StormSoldier/FX/NS_StormSoldier_Lightning
dependencies {
 usedDataInterfaces [&NiagaraDataInterfaceSkeletalMesh]
 usedDynamicInputs [&TransformPosition]
 usedModules [
  &self:SystemSpawnScript.SystemScriptSource.SystemScriptGraph.NiagaraNodeAssignment_0.SetVariables_#1
  &SystemState
  &SampleSkeletalMesh
  &self:SystemSpawnScript.SystemScriptSource.SystemScriptGraph.NiagaraNodeAssignment_1.SetVariables_#2
 ]
}
emitters {Light {rendererClasses [], renderers []}, Long_bolts {rendererClasses [], renderers []}}
summary {
 systemName=NS_StormSoldier_Lightning
 emitters [{emitterName=Long_bolts, rendererClasses []}, {emitterName=Light, rendererClasses []}]
 userVariables []
}
